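; Building actor Blueprint for the Figment air strike that stages incoming effects and triggers the spawning of falling cluster rockets.
; On initialization it plays incoming audio on the air-blast component and activates that component's smoke particles after a short delay.
; A parallel, longer delay syncs the rockets with those effects, then sends the spawn-cluster gameplay event to itself.
; When the replicated StartedDestroying flag is set, it hides and destroys itself.
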(event Custom|HandleBinding
  (params (Actor :type Actor%20Object%20Reference)))

(event Building|EventOnBuildingActorInitialized
  (params (InitializationReason :type EFortBuildingInitializationReason%20Enum :default "StaticallyPlaced") (BuildingPersistentState :type EFortBuildingPersistentState%20Enum :default "Default"))
  ;[1 1 1] Spawn Cluster Bombs On Building Init
  (|CallParentFunction InitializationReason BuildingPersistentState)
  (CallFunction|AirBlastIncomingAudio)
  (Utilities|FlowControl|Delay 1.8) ; Necessary to sync smoke+audio with rockets falling
  (Ability|SendGameplayEventtoActor self (Variables|Default|GetSpawnClusterTag) (Utilities|Struct|MakeGameplayEventData :Instigator (Variables|Actor|GetInstigator) :OptionalObject self :ContextHandle "()" :InstigatorTags "(GameplayTags=)" :TargetTags "(GameplayTags=)" :EventMagnitude 0.0 :TargetData "()")))

(event Custom|AirBlast_Incoming_Audio
  ;[1 1 1] Spawn Cluster Bombs On Building Init
  (Audio|SpawnSoundAttached "/Figment_AirStrike/Audio/SC_Figment_AppleSauce_Incoming_In_Air_Cue.SC_Figment_AppleSauce_Incoming_In_Air_Cue" (Variables|Default|GetP_AirBlast))
  (Utilities|FlowControl|Delay 1.0) ; Necessary to sync audio with smoke particles
  (Components|Activation|Activate (Variables|Default|GetP_AirBlast) true))

(event Custom|DestroyClusterSpawner
  ;[1 1 1] DestroySelf
  (Rendering|SetActorHiddenInGame true)
  (Actor|DestroyActor))

(fn OnRep_bResumeSimulation
  (if (Variables|HitEvent|GetbResumeSimulation)
    _))

(fn OnRep_StartedDestroying
  (if (Variables|Death|GetStartedDestroying)
    (CallFunction|DestroyClusterSpawner)))

(fn ConstructionScript)
