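; Rocket projectile for the Figment air strike that spins a particle trail behind its hidden actor and explodes on stop, bounce or a fuse timer.
; Stop and fuse handlers record the location for DoExplosion, whose cosmetic-only explosion event plays effects and sound, with water-specific audio and feedback for one surface type.
; A separate WarpInDamage event box-traces upward and applies structure damage to damageable actors, skipping Deimos spawners.
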(event Custom|WarpInDamage
  (bind _outhits (Collision|MultiBoxTraceForObjects (+ (Transformation|GetActorLocation) (Math|Vector|MakeVector 0.0 0.0 1.0)) (+ (+ (Transformation|GetActorLocation) (Math|Vector|MakeVector 0.0 0.0 1.0)) (* (Math|Vector|MakeVector 0.0 0.0 1.0) (Variables|Default|GetDamageHeight))) "65.000000,65.000000,65.000000" (Math|Conversions|RotationFromXVector (Math|Vector|MakeVector 0.0 0.0 1.0)) (Variables|Default|GetDestroyObjectTypes)))
  (for _array_element _outhits
    (if (Variables|/Script/Engine.Actor|GetbCanBeDamaged (.HitActor (Collision|BreakHitResult _array_element)))
      (bind _bdidactorhavetags (Class|AthenaQuestFunctionLibrary|AthenaCheckActorHasTags (.HitActor (Collision|BreakHitResult _array_element)) "(GameplayTags=((TagName=\"Deimos.Spawner\")))"))
      (if (not _bdidactorhavetags)
        (Environment|ApplyGlobalEnvironmentGameplayEffecttoActor (.HitActor (Collision|BreakHitResult _array_element)) "/Game/Athena/Environments/Blueprints/DudeBro/GE_HVAC_StructureDamage.GE_HVAC_StructureDamage_C")))))

(event EventBeginPlay
  ;[1 1 1] Begin Play
  (bind _returnvalue (Effects|Components|ParticleSystem|SpawnFXAtLocation (Variables|Default|GetRocket%20Trail%20Template) (Transformation|GetWorldLocation (Variables|Projectile_Rocket|GetMesh)) (Transformation|GetWorldRotation (Variables|Projectile_Rocket|GetMesh))))
  (Variables|Effects|SetRocketTrailPS _returnvalue)
  (Transformation|AddLocalRotation (Variables|Effects|GetRocketTrailPS) (Math|Rotator|MakeRotator (Math|Random|RandomFloatinRange 0.0 360)) false false)
  (Utilities|Time|SetTimerbyFunctionName self "FuseTimerMax" 15)
  (Rendering|SetActorHiddenInGame true))

(event EventTick
  (params (DeltaSeconds :type Float%20%28single-precision%29))
  ;[1 1 1] On Tick, spin the rocket and particle trail
  (Variables|Default|SetTick%20Delta DeltaSeconds)
  (Networking|CanExecuteCosmeticEvents
    (:True
      (if (Components|Activation|IsActive (Variables|Effects|GetRocketTrailPS))
        (Transformation|SetWorldLocation (Variables|Effects|GetRocketTrailPS) (Transformation|GetWorldLocation (Variables|Projectile_Rocket|GetMesh)) false false)
        (Variables|Effects|SetRocketSpin (Math|Rotator|CombineRotators (Math|Rotator|MakeRotator (* (Variables|Default|GetTick%20Delta) (Variables|Effects|GetRocketSpinRate))) (Variables|Effects|GetRocketSpin)))
        (Transformation|AddLocalRotation (Variables|Effects|GetRocketTrailPS) (Variables|Effects|GetRocketSpin) false false)))))

(event EventOnStop
  (params (Hit :type Hit%20Result%20Structure%20%28by%20ref%29))
  (bind (_bblockinghit _binitialoverlap _time _distance _location _impactpoint _normal _impactnormal _physmat) (Collision|BreakHitResult Hit))
  ;[1 1 1] Impact and Explosion Stuff
  (Variables|Effects|SetDecalLocation _impactpoint)
  (Variables|Effects|SetStopLocation (+ _impactpoint (* _impactnormal (Variables|Effects|GetExplosion%20Impact%20Offset))))
  (Variables|Effects|SetStopRotZ (Math|Rotator|MakeRotfromZ _impactnormal))
  (Variables|Default|SetSurfaceType (Variables|/Script/PhysicsCore.PhysicalMaterial|GetSurfaceType _physmat))
  (if (not (Delivery|IsBeingKilled))
    (Delivery|DoExplosion)))

(event Custom|FuseTimerMax
  ;[1 1 1] Impact and Explosion Stuff
  (Variables|Effects|SetTimerMaxReached? true)
  (Variables|Effects|SetStopLocation (Transformation|GetWorldLocation (Variables|Projectile_Rocket|GetMesh)))
  (if (not (Delivery|IsBeingKilled))
    (Delivery|DoExplosion)))

(event EventOnBounce
  (params (Hit :type Hit%20Result%20Structure%20%28by%20ref%29))
  ;[1 1 1] Impact and Explosion Stuff
  (Variables|Effects|SetTimerMaxReached? true)
  (Variables|Effects|SetStopLocation (Transformation|GetWorldLocation (Variables|Projectile_Rocket|GetMesh)))
  (if (not (Delivery|IsBeingKilled))
    (Delivery|DoExplosion)))

(event EventOnExploded
  (params (HitActors :type Array%20of%20Actor%20Object%20References) (HitResults :type Array%20of%20Hit%20Result%20Structures))
  (bind _vfx_dilation (Variables|Default|GetVFX_Dilation))
  (bind _stoplocation (Variables|Effects|GetStopLocation))
  (bind _explosion_generic_ps (Variables|Effects|GetExplosion_Generic_PS))
  (bind _stoprotz (Variables|Effects|GetStopRotZ))
  ;[1 1 1] Impact and Explosion Stuff
  (Networking|CanExecuteCosmeticEvents
    (:True
      (Components|Activation|SetActive (Variables|Effects|GetRocketTrailPS) false false)
      ;;[1 1 1] SingleLineCheckForWorldObjectsSurfaceType
      (switch Utilities|FlowControl|Switch|SwitchonEPhysicalSurface (Variables|Default|GetSurfaceType)
        (:SurfaceType_Default
          _)
        (:SurfaceType1
          _)
        (:SurfaceType2
          _)
        (:SurfaceType3
          _)
        (:SurfaceType4
          _)
        (:SurfaceType5
          _)
        (:SurfaceType6
          _)
        (:SurfaceType7
          _)
        (:SurfaceType8
          (bind _returnvalue (Effects|SpawnEmitteratLocation _explosion_generic_ps _stoplocation _stoprotz))
          (Variables|/Script/CascadeRuntime.ParticleSystemComponent|SetCustomTimeDilation _vfx_dilation _returnvalue)
          (Audio|PlaySoundatLocation (Variables|Sound|GetRocket_Projectile_Explosion_Water_Sound) (Variables|Effects|GetStopLocation))
          (Rendering|SetVisibility (Variables|Projectile_Rocket|GetMesh) false false)
          (Fortnite|ShowSoundIndicatorLocal :Location (Transformation|GetActorLocation) :MaxAudibleDistance 2.75e+03 :IndicatorType "Gunshot" :Instigator self :Tint "(R=0.000000,G=0.000000,B=0.000000,A=0.000000)" :AttenuationMaxRangeMultiplier 1.0 :AttenuationIntensityMultiplier 1.0)
          ;[1 1 1] Shake that camera / Post Process
          (GameplayCue|ExecuteGameplayCue%28Burst%29Local self (Variables|Default|GetFeedbackCue) (Ability|GameplayCue|MakeGameplayCueParameters 0.0)) ; Cam shake + controller vibration
        )
        (:SurfaceType9
          _)
        (:SurfaceType10
          _)
        (:SurfaceType11
          _)
        (:SurfaceType12
          _)
        (:SurfaceType13
          _)
        (:SurfaceType14
          _)
        (:SurfaceType15
          _)
        (:SurfaceType16
          _)
        (:SurfaceType17
          _))
      (bind _returnvalue_2 (Effects|SpawnEmitteratLocation _explosion_generic_ps _stoplocation _stoprotz))
      (Variables|/Script/CascadeRuntime.ParticleSystemComponent|SetCustomTimeDilation _vfx_dilation _returnvalue_2)
      (Audio|PlaySoundatLocation (Variables|Sound|GetRocket_Projectile_Explosion_Sound) (Variables|Effects|GetStopLocation)))))

(event EventDestroyed
  ;[1 1 1] Impact and Explosion Stuff
  (Networking|CanExecuteCosmeticEvents
    (:True
      (Components|Activation|SetActive (Variables|Effects|GetRocketTrailPS) false false))))

(fn ConstructionScript
  (Effects|Components|ParticleSystem|SetTemplate (Variables|Default|GetRibbon_Trail_PSC) (Variables|Effects|GetRibbon_Trail_PS))
  (Effects|Components|ParticleSystem|SetTemplate (Variables|Default|GetExplosion_Generic_PSC) (Variables|Effects|GetExplosion_Generic_PS))
  (Effects|Components|ParticleSystem|SetTemplate (Variables|Default|GetExplosion_Flesh_Damage_PSC) (Variables|Effects|GetExplosion_Flesh_Damage_PS))
  (Transformation|SetRelativeScale3D (Variables|Projectile_Rocket|GetMesh) (Variables|Projectile_Rocket|GetRocketScale))
  (Variables|Effects|SetRocketSpinRate (Math|Random|RandomFloatinRange (* (Variables|Effects|GetRocketSpinRate) -1) (Variables|Effects|GetRocketSpinRate))))
