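; Air strike rocket projectile that homes on a target location and explodes when it stops.
; The server sets lock-on homing, then clears it if the rocket overshoots.
; Hits stop it except on "NoStick" components, which double its velocity, while entering water stops it at the surface.
; Stopping triggers a one-time server explosion with land or water FX, final-rocket audio, feedback cue, sound indicator and optional physics-field radial force.
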(event ProjectileMovement|EventOnHomingPassedTarget
  ;[1 1 1 0.5] Sets Rockets to just continue on their way if they pass past their original target.
  (Networking|SwitchHasAuthority
    (:Authority
      (Projectile|SetHomingStyle (Variables|Movement|GetProjectileMovementComponent) "None" false :!class /Script/FortniteGame.FortProjectileMovementComponent))))

(event EventBeginPlay
  ;[1 1 1] Begin Play
  (CallFunction|InitFlightControls)
  (Utilities|FlowControl|Sequence
    (:then_0
      (Networking|CanExecuteCosmeticEvents
        (:True
          (Components|Activation|Activate (Variables|Default|Get%20RocketTrailPS) false)
          (Rendering|SetRenderCustomDepth (Variables|Default|Get%20RocketTrailPS) true))))
    (:then_1
      (Utilities|FlowControl|Delay 0.01)
      (Projectile|SetHomingData 0 0 (Utilities|Struct|MakeProjectileHomingData (Variables|Default|GetHomingType) (Variables|Default|GetMinTurnSpeed) (Variables|Default|GetMaxTurnSpeed) (Variables|Default|GetRampTime) 1e+03 (Math|Random|RandomIntegerinRange 0 14) 0 0 (Variables|Default|GetTargetLocation)) :!class /Script/FortniteGame.FortProjectileMovementInterface))))

(event EventOnTouched
  (params (OtherActor :type Actor%20Object%20Reference) (OtherComp :type Primitive%20Component%20Object%20Reference) (HitResult :type Hit%20Result%20Structure%20%28by%20ref%29) (bIsOverlap :type Boolean)))

(event OnComponentHit(CapsuleComponent)
  (params (HitComponent :type Primitive%20Component%20Object%20Reference) (OtherActor :type Actor%20Object%20Reference) (OtherComp :type Primitive%20Component%20Object%20Reference) (NormalImpulse :type Vector :default "0, 0, 0") (Hit :type Hit%20Result%20Structure%20%28by%20ref%29))
  (bind (_bblockinghit _binitialoverlap _time _distance _location _impactpoint _normal _impactnormal _physmat _hitactor _hitcomponent) (Collision|BreakHitResult Hit))
  ;[1 1 1] Bounce off "NoBounce"
  (Utilities|FlowControl|DoOnce (Variables|Default|GetIn%20Water)
    (:Completed
      (if (Components|ComponentHasTag _hitcomponent "NoStick")
        (Variables|/Script/Engine.MovementComponent|SetVelocity (* (Variables|/Script/Engine.MovementComponent|GetVelocity (Variables|Movement|GetProjectileMovementComponent)) 2.0) (Variables|Movement|GetProjectileMovementComponent))
        (else
          (Variables|Projectile|SetbStopSimulatingOnHit true))))))

(event OnEnterWaterEvent(WaterInteractionComponent)
  (params (WaterBody :type Fort%20Water%20Body%20Actor%20Object%20Reference) (WaterInteractionComponent :type Fort%20Water%20Interaction%20Component%20Object%20Reference) (bIsFirstBody :type Boolean))
  ;[1 1 1] New Water Logic
  (Variables|Default|SetIn%20Water true)
  (bind _outwatersurfaceinfo (WaterBody|GetWaterSurfaceInfoStruct WaterBody (Transformation|GetActorLocation) true false false))
  (Game|Components|ProjectileMovement|StopSimulating (Variables|Movement|GetProjectileMovementComponent) (Collision|MakeHitResult false false 0.0 0.0 (Utilities|Struct|BreakWaterSurfaceInfo _outwatersurfaceinfo) "0, 0, 0" (.WaterPlaneNormal (Utilities|Struct|BreakWaterSurfaceInfo _outwatersurfaceinfo)) "0,0,1" 0 WaterBody)))

(event EventOnStop
  (params (Hit_bBlockingHit :type Boolean) (Hit_bInitialOverlap :type Boolean) (Hit_Time :type Float%20%28single-precision%29) (Hit_Distance :type Float%20%28single-precision%29) (Hit_Location :type Vector :default "0, 0, 0") (Hit_ImpactPoint :type Vector :default "0, 0, 0") (Hit_Normal :type Vector :default "0, 0, 0") (Hit_ImpactNormal :type Vector :default "0, 0, 0") (Hit_PhysMat :type Physical%20Material%20Object%20Reference) (Hit_HitActor :type Actor%20Object%20Reference) (Hit_HitComponent :type Primitive%20Component%20Object%20Reference) (Hit_HitBoneName :type Name :default "None") (Hit_BoneName :type Name :default "None") (Hit_HitItem :type Integer) (Hit_ElementIndex :type Integer) (Hit_FaceIndex :type Integer) (Hit_TraceStart :type Vector :default "0, 0, 0") (Hit_TraceEnd :type Vector :default "0, 0, 0"))
  ;[1 1 1] Impact and Explosion Stuff
  (Networking|SwitchHasAuthority
    (:Authority
      (Utilities|FlowControl|DoOnce false
        (:Completed
          (Delivery|DoExplosion)
          (CallFunction|ExplodeFX Hit_Location Hit_Normal))))
    (:Remote
      _))
  (Development|SetHiddeninGame (Variables|Collision|GetCapsuleComponent) true false)
  (Components|Activation|Deactivate (Variables|Default|Get%20RocketTrailPS)))

(event Custom|ExplodeFX
  (params (Location :type Vector) (Normal :type Vector))
  (bind _returnvalue (Math|Rotator|MakeRotfromZ Normal))
  ;[1 1 1] Impact and Explosion Stuff
  (Audio|PlaySoundatLocation (Variables|Sound|GetRocket_Projectile_Explosion_Sound) Location)
  (if (Rendering|WasActorRecentlyRendered 0.2)
    (if (Variables|Default|GetIn%20Water)
      (bind _returnvalue_2 (Effects|Components|ParticleSystem|SpawnFXAtLocation "/Game/Athena/Items/Consumables/AirStrike/P_Athena_AppleSauce_Explosion_Water.P_Athena_AppleSauce_Explosion_Water" Location _returnvalue "1.000000,1.000000,1.000000" true "AutoRelease"))
      (Effects|Components|ParticleSystem|SetFloatParameter _returnvalue_2 "Global Scale" 0.9)
      (Effects|Components|ParticleSystem|SetColorParameter _returnvalue_2 "Explosion Smoke Color" "(R=0.526000,G=0.540000,B=0.567000,A=0.500000)")
      (else
        (bind _returnvalue_3 (Effects|Components|ParticleSystem|SpawnFXAtLocation "/Game/Athena/Items/Consumables/AirStrike/P_Impact.P_Impact" Location _returnvalue "1.000000,1.000000,1.000000" true "AutoRelease"))
        (Effects|Components|ParticleSystem|SetFloatParameter _returnvalue_3 "Global Scale" 0.9)
        (Effects|Components|ParticleSystem|SetColorParameter _returnvalue_3 "Explosion Smoke Color" "(R=0.526000,G=0.540000,B=0.567000,A=0.500000)"))))
  (Audio|Components|Audio|FadeOut (Variables|Default|GetAudio_Proj_Loop) 0.1 0.0 "Linear" :!class /Script/Engine.AudioComponent)
  (if (Variables|Default|GetFinalRocket)
    (Utilities|FlowControl|DoOnce false
      (:Completed
        (Audio|PlaySoundatLocation (Variables|Default|GetFinalRocketExplosionSound) Location)))))

(event EventOnExploded
  (params (HitActors :type Array%20of%20Actor%20Object%20References) (HitResults :type Array%20of%20Hit%20Result%20Structures))
  ;[1 1 1] Impact and Explosion Stuff
  (Fortnite|ShowSoundIndicatorLocalforTeam :Location (Transformation|GetActorLocation) :MaxAudibleDistance (Variables|/Script/Engine.SoundBase|GetMaxDistance (Variables|Sound|GetRocket_Projectile_Explosion_Sound)) :IndicatorType "Gunshot" :Team (Team|GetActorTeam self) :Affiliations (Utilities|Array|MakeArray "Hostile" "Neutral" "Friendly") :Tint "(R=0.000000,G=0.000000,B=0.000000,A=0.000000)" :AttenuationMaxRangeMultiplier 1.0 :AttenuationIntensityMultiplier 1.0)
  ;[1 1 1] Shake that camera
  (GameplayCue|ExecuteGameplayCue%28Burst%29Local self (Variables|Default|GetFeedbackCue) (Ability|GameplayCue|MakeGameplayCueParameters 0.0)) ; Cam shake + controller vibration
  (Audio|Components|Audio|FadeOut (Variables|Default|GetAudio_Proj_Loop) 1 0.0 "Linear" :!class /Script/Engine.AudioComponent)
  (Utilities|AllowJanus
    (:Allow
      (bind _returnvalue (Development|GetConsoleVariableIntValue "r.PhysicsField.EnableField"))
      (if (== _returnvalue 1)
        (Game|SpawnActorfromClass "/Game/Effects/Niagara/Blueprints/PhsyicsFields/FN_RadialForce.FN_RadialForce_C" (Transformation|GetWorldTransform (Variables|Collision|GetCapsuleComponent)) "Undefined" "MultiplyWithRoot" 0 0.5 0.8 1.2e+03 2.0 0.0 0)))))

(event OnSystemFinished(RocketTrailPS)
  (params (PSystem :type Cascade%20Particle%20System%20Component%20Object%20Reference))
  ;[1 1 1] Impact and Explosion Stuff
  ;;[0 0 0] Keep the trail playing after explode
  (Rendering|SetActorHiddenInGame true))

(fn InitFlightControls
  (bind _projectilemovementcomponent (Variables|Movement|GetProjectileMovementComponent))
  ;[1 1 1] Target strike straight on to target location
  (Networking|SwitchHasAuthority
    (:Authority
      (Projectile|SetHomingStyle _projectilemovementcomponent "LockOn" false :!class /Script/FortniteGame.FortProjectileMovementComponent)
      (Projectile|SetHomingTargetPosition _projectilemovementcomponent (Variables|Default|GetTargetLocation) :!class /Script/FortniteGame.FortProjectileMovementComponent))
    (:Remote
      _))
  (Variables|Default|SetMinTurnSpeed 1e+02)
  (Variables|Default|SetMaxTurnSpeed 1e+02))

(fn OnRep_EXP_N)

(fn ConstructionScript)
