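; Event-triggered gameplay ability for the Figment AirStrike (AppleSauce) that rains explosive drunk rockets from an airborne spawner actor.
; After a short delay, the server loops with randomized intervals, spawning rockets around and progressively lower beneath the spawner to fake falling.
; Each rocket gets primary and environment explosion effects plus a random downward target within a radius; the last is flagged final.
; A radial SpawnClusters routine exists but is not called here.
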(event Ability|EventActivateAbilityFromEvent
  (params (EventData :type Gameplay%20Event%20Data%20Structure%20%28by%20ref%29))
  (bind (_eventtag _instigator _target _optionalobject) (Utilities|Struct|BreakGameplayEventData EventData))
  (|CallParentFunction EventData)
  ;[1 1 1] Set Balance using sheets.
  (if (Variables|GAVars|Rows|GetUseBalanceRows)
    (CallFunction|SetBalance)
    (else
      _))
  ;[1 1 1] Save Cluster Spawner
  (bind _asb_bga_figment_apple_sauce_rocket_spawner (Utilities|Casting|CastToB_BGA_Figment_AppleSauce_RocketSpawner _optionalobject))
  (Variables|Default|SetRocketSpawner _asb_bga_figment_apple_sauce_rocket_spawner)
  ;[1 1 1] Save out spec for later use
  (Variables|GAVars|Damage|SetGECS_Explode (Ability|MakeGameplayEffectContainerSpec (Variables|GAVars|Damage|GetT_Targeting)))
  (Variables|Default|SetGECS%20Explode%20Additional (Ability|MakeGameplayEffectContainerSpec (Variables|Default|GetEnv%20Targeting)))
  (if (Networking|IsServer)
    (Variables|Default|SetRandomX (Math|Random|RandomFloatinRange -5e+02 5e+02))
    (Variables|Default|SetRandomY (Math|Random|RandomFloatinRange -5e+02 5e+02))
    (Utilities|FlowControl|Delay 0.25)
    (bind _index (Utilities|ForLoopwithDelay 1 (Math|Float|Truncate (ScalableFloat|GetValueatLevel (Variables|Default|GetSF_ClusterAmount))) (ScalableFloat|GetValueatLevel (Variables|Default|GetMinDelayBetweenRocket)) (ScalableFloat|GetValueatLevel (Variables|Default|GetMaxDelayBetweenRocket)))
      (:LoopBody
        (bind _returnvalue (Projectile|SpawnProjectile :ProjectileClass (Variables|GAVars|RocketSpawn|GetPrj_Rocket) :RequestedBy (Ability|GetAvatarActorfromActorInfo) :SpawnLocation (+ (Transformation|GetActorLocation (Variables|Default|GetRocketSpawner)) (Math|Vector|MakeVector (Math|Random|RandomFloatinRange -1e+03 1e+03) (Math|Random|RandomFloatinRange -1e+03 1e+03) (- (Math|Random|RandomFloatinRange -6e+02 -7e+02) (* _index (Variables|Default|GetFakeFallValue))))) :SpawnRotation (Math|Rotator|MakeRotator (.roll (Variables|Default|GetRotation)) (.pitch (Variables|Default|GetRotation)) (Math|Random|RandomFloatinRange 0.0 3.6e+02)) :SpawnDirection (Math|Rotator|MakeRotator (.roll (Variables|Default|GetRotation)) (.pitch (Variables|Default|GetRotation)) (Math|Random|RandomFloatinRange 0.0 3.6e+02)) :EffectContainerSpecToApplyOnExplode (Variables|GAVars|Damage|GetGECS_Explode) :bUseDefaultPhysics false :InitialSpeed (Variables|Default|GetSpeed) :GravityScale 1.0 :ChargePercent 0.0 :OptionalAssociatedItemDef "/Game/Athena/Items/Consumables/AirStrike/Athena_AppleSauce.Athena_AppleSauce"))
        (Delivery|AddAdditionalExplodeEffectContainer _returnvalue (Variables|Default|GetGECS%20Explode%20Additional))
        (bind _asb_prj_figment_apple_sauce_drunk_rocket (Utilities|Casting|CastToB_Prj_Figment_AppleSauce_DrunkRocket _returnvalue))
        (Variables|/Figment_AirStrike/Gameplay/B_Prj_Figment_AppleSauce_DrunkRocket.B_Prj_Figment_AppleSauce_DrunkRocket_C|SetTargetLocation (+ (Math|Vector|MakeVector (Math|Random|RandomFloatinRange (* (ScalableFloat|GetValueatLevel (Variables|Default|GetRadius)) -1.0) (ScalableFloat|GetValueatLevel (Variables|Default|GetRadius))) (Math|Random|RandomFloatinRange (* (ScalableFloat|GetValueatLevel (Variables|Default|GetRadius)) -1.0) (ScalableFloat|GetValueatLevel (Variables|Default|GetRadius))) (* (ScalableFloat|GetValueatLevel (Variables|Default|GetHeight)) -1.0)) (Transformation|GetActorLocation (Variables|Default|GetRocketSpawner))) _asb_prj_figment_apple_sauce_drunk_rocket)
        (if (== _index (Math|Float|Truncate (ScalableFloat|GetValueatLevel (Variables|Default|GetSF_ClusterAmount))))
          (Variables|/Figment_AirStrike/Gameplay/B_Prj_Figment_AppleSauce_DrunkRocket.B_Prj_Figment_AppleSauce_DrunkRocket_C|SetFinalRocket true _asb_prj_figment_apple_sauce_drunk_rocket))))))

(event Custom|FinishedSpawning
  ;[1 1 1] End Ability and Destroy Cluster Spawner
  (Variables|/Figment_AirStrike/Gameplay/B_BGA_Figment_AppleSauce_RocketSpawner.B_BGA_Figment_AppleSauce_RocketSpawner_C|SetStartedDestroying true (Variables|Default|GetRocketSpawner))
  (Ability|EndAbility))

(fn AddVariance
  (params (Default :type Float%20%28double-precision%29) (Variance :type Float%20%28double-precision%29))
  (return (Math|Random|RandomFloatinRange (- Default Variance) (+ Default Variance))))

(fn SetBalance
  (Variables|GAVars|RocketSpawn|SetBottleRocketSpeed (ScalableFloat|GetValueatLevel (Variables|GAVars|Rows|GetRow_RocketSpeed)))
  (Variables|GAVars|RocketSpawn|SetBottleRocketSpeedVariance (ScalableFloat|GetValueatLevel (Variables|GAVars|Rows|GetRow_RocketSpeedVariance)))
  (Variables|GAVars|RocketSpawn|SetSpawnDelay (ScalableFloat|GetValueatLevel (Variables|GAVars|Rows|GetRow_SpawnDelay)))
  (Variables|GAVars|RocketSpawn|SetSpawnDelayVarience (ScalableFloat|GetValueatLevel (Variables|GAVars|Rows|GetRow_SpawnDelayVariance)))
  (Variables|GAVars|RocketSpawn|SetMaxShots (ScalableFloat|GetValueatLevel (Variables|GAVars|Rows|GetRow_MaxShots)))
  (Variables|GAVars|RocketSpawn|SetSpread (ScalableFloat|GetValueatLevel (Variables|GAVars|Rows|GetRow_Spread))))

(fn SpawnClusters
  (bind _sf_clusteramount (Variables|Default|GetSF_ClusterAmount))
  (bind _returnvalue (ScalableFloat|GetValueatLevel _sf_clusteramount))
  ;[1 1 1] Spawn clusters on server
  (for _index (range 1 (+ (Math|Float|Truncate _returnvalue) 1))
    (bind _returnvalue_2 (Projectile|SpawnProjectile :ProjectileClass (Variables|GAVars|RocketSpawn|GetPrj_Rocket) :RequestedBy (Ability|GetAvatarActorfromActorInfo) :SpawnLocation (Transformation|GetActorLocation (Variables|Default|GetRocketSpawner)) :SpawnRotation (Transformation|GetActorRotation (Variables|Default|GetRocketSpawner)) :SpawnDirection (Math|Rotator|CombineRotators (Transformation|GetActorRotation (Variables|Default|GetRocketSpawner)) (Math|Rotator|MakeRotator 0.0 (ScalableFloat|GetValueatLevel (Variables|Default|GetSF_LaunchAngle)) (* _index (/ 3.6e+02 _returnvalue)))) :EffectContainerSpecToApplyOnExplode (Variables|GAVars|Damage|GetGECS_Explode) :bUseDefaultPhysics false :InitialSpeed (Math|Random|RandomFloatinRange 1e+02 1.4e+03) :GravityScale 0.5 :ChargePercent 0.0))
    (Delivery|AddAdditionalExplodeEffectContainer _returnvalue_2 (Variables|Default|GetGECS%20Explode%20Additional)))
  ;[1 1 1] Once clusters have been spawned, Destroy BGA and End Ability
  (CallFunction|FinishedSpawning))
